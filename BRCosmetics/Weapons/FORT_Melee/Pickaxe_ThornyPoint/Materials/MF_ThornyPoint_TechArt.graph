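# Material function for the Thorny Point pickaxe that blends an animated, view-dependent color effect into base color and emissive within a masked area.
# A dot product of a fixed local-space vector and camera vector, plus panning swirl noise distorted by Perlin noise, is wrapped into a color ramp.
# Time-of-day tint parameters feed a named color reroute.
# The mask input's green channel, zeroed below a threshold, supplies the blend mask.
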MaterialFunction /BRCosmetics/Weapons/FORT_Melee/Pickaxe_ThornyPoint/Materials/MF_ThornyPoint_TechArt
graph {nodes [
  n1 = NamedRerouteUsage()
  n2 = NamedRerouteUsage()
  FunctionOutput(None=BlendMaterialAttributes(
    A=NamedRerouteUsage()
    B=SetMaterialAttributes!(MaterialAttributes=@n1, BaseColor=@n2, EmissiveColor=@n2) {attributeSetTypes [#1, #2]}
    Alpha=NamedRerouteUsage()
   ))
  n3 = ComponentMask(None=FunctionInput() {inputName=In_TechArtMask, sortPriority=1}) {g=true}
  NamedRerouteDeclaration(None=If(A=@n3, "A > B"=@n3, "A < B"=Constant()) {constB=0.35}) {name=In_TechArtMask, nodeColor=<rgba 1 0 0.6562 1>}
  n4 = Transform(None=Constant3Vector() {constant=<rgba 2 0.5 1.5 1>}) {transformSourceType=TRANSFORMSOURCE_Local}
  n5 = CameraVectorWS()
  n6 = Multiply(A=TextureCoordinate())
  n7 = Constant() {r=1}
  n8 = Panner(Coordinate=Multiply(A=TextureCoordinate(), B=AppendVector(A=@n7, B=@n7)) {constB=10}, Speed=Constant2Vector() {g=0.05})
  n9 = Multiply(
   A=ConstantBiasScale(None=TextureSample(UVs=@n8) {samplerType=SAMPLERTYPE_Masks, texture=/BRCosmetics/Weapons/FORT_Melee/Pickaxe_SkilledSkull/Textures/T_PerlinNoise_RGB}.B) {bias=-0.5, scale=2}
   B=ScalarParameter() {defaultValue=0.1, group=Blade, parameterName=DistortionNoiseScale}
  )
  n10 = TextureSample(UVs=Panner(Coordinate=Add(A=@n6, B=@n9), Speed=Constant2Vector() {g=0.1})) {samplerType=SAMPLERTYPE_LinearColor, texture=/BRCosmetics/Weapons/FORT_Melee/Pickaxe_SkilledSkull/Textures/T_SwirlNoise}
  n11 = ScalarParameter() {defaultValue=0.1, group=Blade, parameterName=EdgeNoiseScale}
  n12 = Add(
   A=Multiply(A=DotProduct(A=@n4, B=@n5), B=Constant() {r=0.33})
   B=MaterialFunctionCall(Input=Multiply(A=@n10.R, B=@n11), Input Low=Constant(), Input High=Constant() {r=0.1}) {materialFunction=&RemapValue_0-1}.Result
  )
  n13 = ComponentMask(None=ColorRamp!(None=MaterialFunctionCall!(Value=@n12, Wrap) {materialFunction=/Engine/Functions/Engine_MaterialFunctions03/Math/WrapFloat}.Result) {colorCurve="<self>.ColorCurve"}) {b=true, g=true, r=true}
  n14 = NamedRerouteUsage()
  NamedRerouteDeclaration(None=LinearInterpolate(A=@n13, B=MaterialFunctionCall!(Base=@n13, Add=@n14) {materialFunction=/Engine/Functions/Engine_MaterialFunctions02/Math/MultiplyAdd}.Result)) {name=TechArt, nodeColor=<rgba 0.0547 0 1 1>}
  n15 = FunctionInput() {inputName=In_MaterialAttributes, inputType=FunctionInput_MaterialAttributes}
  NamedRerouteDeclaration(None=GetMaterialAttributes(None=NamedRerouteDeclaration(None=@n15) {name=InMat, nodeColor=<rgba 0 0.8125 1 1>}) {attributeGetTypes [#3]}.Normal) {name=Normal, nodeColor=<rgba 0 0.9297 1 1>}
  NamedRerouteDeclaration(None=MaterialFunctionCall(
    Color_Morning=VectorParameter() {group=TOD Tints, parameterName=Color_Morning, defaultValue=<rgba 0.8 0.8 0.8 1>}.RGB
    Color_Day=VectorParameter() {group=TOD Tints, parameterName=Color_Day, defaultValue=<rgba 1 1 1 1>}.RGB
    Color_Evening=VectorParameter() {group=TOD Tints, parameterName=Color_Evening, defaultValue=<rgba 0.9 0.6938 0.675 1>}.RGB
    Color_Night=VectorParameter() {group=TOD Tints, parameterName=Color_Night, defaultValue=<rgba 0.4333 0.4 0.8 1>}.RGB
   ) {materialFunction=&Color_Time_of_Day}.Result) {name=TimeColor, nodeColor=<rgba 0 0.6016 1 1>}
 ]}
